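# Sound cue for the smart rifle's auto-aim lock shot, named as a layer of the weapon's firing audio.
# It plays a single wave through a modulator that holds pitch fixed.
# It uses the first-person ranged-fire sound class with close-projectile attenuation at reduced volume.
# A concurrency override caps simultaneous instances at a low count, stopping the oldest first.
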SoundCue /SmartRifle/Audio/AutoAimRifle_Lock_Shot_Layer
root=Modulator_2
nodes [
 {id=Modulator_2, children [WavePlayer_2], properties {pitchMax=1, pitchMin=1}}
 {id=WavePlayer_2, properties {soundWaveAssetPtr=&/AutoAimRifle_Lock_Shot}}
]
properties {
 attenuationSettings=&Fort_Weapon_Projectile_Close_Athena_Att
 bOverrideConcurrency=true
 maxDistance=4950
 soundClassObject=&Fort_WeaponRanged_Fire_1P_PSM
 volumeMultiplier=0.35
 attenuationOverrides {falloffDistance=4100, attenuationShapeExtents {x=900}}
 concurrencyOverrides {maxCount=2, resolutionRule=StopOldest}
}
