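# Sound cue for the SmartRifle auto-aim target count sound, classed as weapon foley.
# A mixer layers two waves: a target-count wave with slight random volume variation, and a quieter lock tick at a fixed lowered pitch.
# It uses weapon reload attenuation and overrides concurrency to prevent new instances once the limit is reached.
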SoundCue /SmartRifle/Audio/AutoAimRifle_Target_Count_Cue
root=Mixer_0
nodes [
 {id=Mixer_0, children [Modulator_0,Modulator_1], properties {inputVolume [1,0.5]}}
 {id=Modulator_0, children [WavePlayer_0], properties {pitchMax=1, pitchMin=1, volumeMin=0.85}}
 {id=Modulator_1, children [WavePlayer_1], properties {pitchMax=0.7, pitchMin=0.7, volumeMax=1, volumeMin=1}}
 {id=WavePlayer_0, properties {soundWaveAssetPtr=/Smartgun/Gameplay/StackingLockons/Sounds/AutoAim_Target_Count_A_01}}
 {id=WavePlayer_1, properties {soundWaveAssetPtr=&/AutoAimRifle_Lock_Tick}}
]
properties {
 attenuationSettings=&Fort_Weapon_Reload_Att
 bOverrideConcurrency=true
 maxDistance=2250
 soundClassObject=&Fort_WeaponFoley
 volumeMultiplier=0.476
 attenuationOverrides {falloffDistance=4100, attenuationShapeExtents {x=900}}
 concurrencyOverrides {resolutionRule=PreventNew}
}
