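# Sound cue for the Smart Rifle's auto-aim, named for the moment a target lock appears.
# A single lock chevron wave feeds a modulator with fixed unit pitch and volume, so playback has no random variation.
# The cue lowers pitch and volume, uses gadget item sound class and borrows the security turret's sphere attenuation.
# A concurrency override allows one instance per owner, stopping the oldest.
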SoundCue /SmartRifle/Audio/AutoAimRifle_Target_Lock_Appears_Cue
root=Modulator_1
nodes [
 {id=Modulator_1, children [WavePlayer_1], properties {pitchMax=1, pitchMin=1, volumeMax=1, volumeMin=1}}
 {id=WavePlayer_1, properties {soundWaveAssetPtr=&/AutoAimRifle_Lock_Chevron}}
]
properties {
 attenuationSettings=/Game/Athena/Sounds/Gadgets/SecurityTurret/Security_Turret_Sphere_Att
 bOverrideConcurrency=true
 maxDistance=3500
 pitchMultiplier=0.85
 soundClassObject=&Fort_Gadgets_Items
 volumeMultiplier=0.1
 concurrencyOverrides {bLimitToOwner=true, maxCount=1, resolutionRule=StopOldest}
}
