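# Sound cue for the auto-aim rifle's target lock sound, routed to the gadgets and items sound class.
# A mixer layers a chevron lock wave with a second target-lock wave that passes through a fixed-volume modulator and a short fixed delay.
# Both layers are scaled down, the first by mixer input volume and the second by the modulator.
# It borrows the security turret's attenuation asset and overrides concurrency to stop the quietest instance.
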SoundCue /SmartRifle/Audio/AutoAimRifle_Target_Lock_Lock_Cue
root=Mixer_1
nodes [
 {id=Delay_0, children [Modulator_0], properties {delayMax=0.15, delayMin=0.15}}
 {id=Mixer_1, children [WavePlayer_4,Delay_0], properties {inputVolume [0.3,1]}}
 {id=Modulator_0, children [WavePlayer_5], properties {pitchMax=1, pitchMin=1, volumeMax=0.3, volumeMin=0.3}}
 {id=WavePlayer_4, properties {soundWaveAssetPtr=&/AutoAimRifle_Lock_Chevron}}
 {id=WavePlayer_5, properties {soundWaveAssetPtr=&/AutoAim_Target_Lock_Lock_03}}
]
properties {
 attenuationSettings=/Game/Athena/Sounds/Gadgets/SecurityTurret/Security_Turret_Sphere_Att
 bOverrideConcurrency=true
 maxDistance=3500
 soundClassObject=&Fort_Gadgets_Items
 volumeMultiplier=0.7
 attenuationOverrides {
  bEnableFocusInterpolation=true
  bEnableListenerFocus=true
  falloffDistance=1200
  nonFocusVolumeAttenuation=0.5
  stereoSpread=500
  attenuationShapeExtents {x=0}
 }
 concurrencyOverrides {maxCount=3, resolutionRule=StopQuietest}
}
