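# Looping mono MetaSound source that plays blink beeps, named for the Auto Aim Rifle's locked-on lights, as weapon foley.
# On play, a Trigger Repeat with a fixed half-second period retriggers a Wave Player using the Wave Asset input.
# A BlinkSpeed divide and float-to-time conversion exists but is not wired to the repeat period, so that input does not affect the beep rate.
# Foley attenuation and owner-limited concurrency govern playback.
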MetaSoundSource /SmartRifle/Audio/MSS_AutoAimRifle_WeaponLockedOn_LightsBlink
inputs [
 BlinkSpeed {type=Float}
 Time.ElapsedTime {type=Float}
 Time.GameTimeSeconds {type=Float}
 Time.PlayTime {type=Float}
 Time.TimeSinceLastPlayed {type=Float}
 UE.Source.OnPlay {type=Trigger}
 Wave Asset {type=WaveAsset}
]
interfaces [Fort.Time:1.0, UE.OutputFormat.Mono:1.0, UE.Source:1.1]
outputs ["UE.OutputFormat.Mono.Audio:0" {type=Audio}]
pages [Default {
  input_defaults {
   BlinkSpeed=2
   Time.ElapsedTime=0
   Time.GameTimeSeconds=0.396
   Time.PlayTime=0
   Time.TimeSinceLastPlayed=0
   UE.Source.OnPlay=null
   Wave Asset=&/AutoAimRifle_Lock_Blink_Beeps
  }
  nodes [
   UE.ConversionFloatToTime(In=UE.Divide:Float(AdditionalOperands=^BlinkSpeed) {also {defaults {PrimaryOperand=0.5}}}.Out)
   InitVariable.Trigger()
   WavePlayer_0 = UE.Wave Player:Mono(Wave Asset=^Wave Asset, Play=UE.TriggerRepeat(Start=^UE.Source.OnPlay) {also {defaults {Period=0.5}}}.RepeatOut) {also {version [1,1]}}
  ]
  links [{dst="@UE.OutputFormat.Mono.Audio:0", src=WavePlayer_0.Out Mono}]
 }]
properties {
 assetImportData=&self:AssetImportData
 attenuationSettings=&AS_Fort_WeaponFoley
 bLooping=true
 duration=10000
 maxDistance=2250
 sampleRate=48000
 soundClassObject=&Fort_WeaponFoley
 volume=0.3
 concurrencyOverrides {bLimitToOwner=true}
 concurrencySet [&CS_Fort_WeaponFoley]
}
